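# Conversation database for the Binary NPC in the NPC library, providing its dialogue under the Conversation.Entry.Binary tag.
# The entry leads to one Speech node with a default line and contextual lines for a first-ever conversation or specific player skins (CIDs).
# The first-meeting line takes its priority from a curve row, while the skin lines have zero priority.
# The speech then branches through link nodes to the shared NPC Library quests and basic services conversations.
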ConversationDatabase /NPCLibrary/NPCs/Binary/Binary/NPCLibrary_Conversation_Binary
entry_points [{tag=Conversation.Entry.Binary, destinations [#1]}]
nodes [
 ConversationEntryPointNode_7 {
  id=#1
  class=ConversationEntryPointNode
  outputs [#2]
  properties {entryTag {tagName=Conversation.Entry.Binary}}
 }
 Speech_C_4 {
  id=#2
  class=Speech_C
  outputs [#3, #4]
  properties {
   nodeName=""
   generalConfig {
    defaultMessage="My past? Nah, doesn't bother me... Most of the time..."
    contextualMessages [
     {
      message="You kicked serious butt out there. Don't tell anyone I said that! "
      contextRequirements [{requirement {class=FortControllerRequirement_IsFirstEverConversationWithNPC}}]
      priority {curve {rowName=Tandem.Speech.Priority.FirstEverInteraction}}
     }
     {
      message="I missed talking to you, but my question still stands. Who are we?"
      contextRequirements [{requirement {
         class=FortControllerRequirement_HasCID
         properties {allowedCIds [/BRCosmetics/Athena/Items/Cosmetics/Characters/CID_A_395_Athena_Commando_F_BinaryTwin]}
        }}]
      priority {value=0, curve {curveTable=null, rowName=None}}
     }
     {
      message=Maybe we should team up more often.
      contextRequirements [{requirement {
         class=FortControllerRequirement_HasCID
         properties {allowedCIds [/BRCosmetics/Athena/Items/Cosmetics/Characters/CID_540_Athena_Commando_M_MeteorManRemix]}
        }}]
      priority {value=0, curve {curveTable=null, rowName=None}}
     }
     {
      message="Hey, penpal! You're shorter than I remember."
      contextRequirements [{requirement {
         class=FortControllerRequirement_HasCID
         properties {allowedCIds [/BRCosmetics/Athena/Items/Cosmetics/Characters/CID_547_Athena_Commando_F_Meteorwoman]}
        }}]
      priority {value=0, curve {curveTable=null, rowName=None}}
     }
     {
      message="Admit it. You like me. "
      contextRequirements [{requirement {
         class=FortControllerRequirement_HasCID
         properties {allowedCIds [/BRCosmetics/Athena/Items/Cosmetics/Characters/CID_A_373_Athena_Commando_M_OriginPrisoner]}
        }}]
      priority {value=0, curve {curveTable=null, rowName=None}}
     }
     {
      message="The Loopers keep asking me when you're coming to the party!"
      contextRequirements [{requirement {
         class=FortControllerRequirement_HasCID
         properties {allowedCIds [/BRCosmetics/Athena/Items/Cosmetics/Characters/CID_A_286_Athena_Commando_M_Turtleneck]}
        }}]
      priority {value=0, curve {curveTable=null, rowName=None}}
     }
     {
      message="Rumor has it you're settling down. Gonna grow corn, Farmer Steel?"
      contextRequirements [{requirement {
         class=FortControllerRequirement_HasCID
         properties {allowedCIds [/BRCosmetics/Athena/Items/Cosmetics/Characters/CID_140_Athena_Commando_M_Visitor]}
        }}]
      priority {value=0, curve {curveTable=null, rowName=None}}
     }
    ]
   }
  }
 }
 ConversationLinkNode_0 {id=#4, class=ConversationLinkNode, properties {remoteEntryTag {tagName=Conversation.Entry.BasicServices}}}
 ConversationLinkNode_1 {id=#3, class=ConversationLinkNode, properties {remoteEntryTag {tagName=Conversation.Entry.NPCLibrary.Quests}}}
]
